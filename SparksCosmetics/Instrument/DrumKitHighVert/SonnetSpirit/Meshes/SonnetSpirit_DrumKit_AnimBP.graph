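# Animation Blueprint for the SonnetSpirit drum kit cosmetic, animating the drum pieces as characters with faces.
# A music clock, faked briefly in previews, switches between music and no-music state machines, each with idle and fidget states, layered per bone over the input pose.
# Material scalar parameters drive six faces, randomized during music and blinking otherwise.
# Drum events from the player's linked anim layer stamp hit times into the material.
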AnimBlueprint /SparksCosmetics/Instrument/DrumKitHighVert/SonnetSpirit/Meshes/SonnetSpirit_DrumKit_AnimBP
target_skeleton=&/SonnetSpirit_DrumKit_Skeleton
graphs [
 AnimGraph {nodes [
   AnimGraphNode_StateMachine_0 = Main_NoMusic() {is=title AnimGraphNode_StateMachine}
   AnimGraphNode_StateMachine_1 = Main() {is=title AnimGraphNode_StateMachine}
   n1 = Get ClockValid() -> (ClockValid:bool)
   n2 = Get FakeClockValid() -> (FakeClockValid:bool)
   n3 = Layered blend per bone(
    BasePose=Input Pose()
    BlendPoses_0=Blend Poses by bool(
     BlendPose_0=@AnimGraphNode_StateMachine_1
     BlendPose_1=@AnimGraphNode_StateMachine_0
     BlendTime_0=0.5
     BlendTime_1=0.5
     bActiveValue=OR Boolean(A=@n1, B=@n2)
    )
   ) {layerSetup [{branchFilters [
       {boneName=floortom_nickolas}
       {boneName=snare_anthea}
       {boneName=tom_mel}
       {boneName=tom_ryan}
       {boneName=hihat_luca_arm_r_01}
       {boneName=hihat_luca_arm_l_01}
       {boneName=ride_arm_r_01}
       {boneName=ride_arm_l_01}
      ]}]}
   Output Pose(Result=@n3)
   n4 = Get BaseAnim() -> (BaseAnim:Anim Sequence Base Object Reference)
   Save cached pose 'ChacheIdle'(Pose=Blend Poses by bool(
     BlendPose_0=Play Beat Synced Animation!(
      ExplicitTime:float=0
      bShouldLoop:bool=True
      InSequence:Anim Sequence Base Object Reference=@n4
      MusicClock:Music Clock Object Reference
      TempoMultiplier:float=1
      bEnableTempoVariants:bool=False
      TempoVariantHysteresis:float=0.05
      bShowDebug:bool=False
     ) -> (Pose:Pose Link Structure) {is=title AnimGraphNode_PlayBeatSyncedAnim}
     BlendPose_1=Sequence Player!(Sequence:Anim Sequence Base Object Reference=@n4) -> (Pose)
     BlendTime_0=0
     BlendTime_1=0
     bActiveValue=Get ClockValid()
    )) {cacheName=ChacheIdle}
  ]}
 Main_NoMusic {
  entry=Idle
  path=AnimGraph.AnimGraphNode_StateMachine_0.Main_NoMusic
  ids=id
  nodes [
   AnimStateNode_0 = Idle()
   AnimStateNode_1 = Play() {is=title AnimStateNode}
   AnimStateTransitionNode_0 = AnimStateTransitionNode() {bAllowInertializationForSelfTransitions=true}
   AnimStateTransitionNode_1 = AnimStateTransitionNode() {bAllowInertializationForSelfTransitions=true}
  ]
  links [
   {
    dst=Play
    rule=AnimGraph.AnimGraphNode_StateMachine_0.Main_NoMusic.AnimStateTransitionNode_0.Transition
    src=Idle
    via=AnimStateTransitionNode_0
   }
   {
    dst=Idle
    rule=AnimGraph.AnimGraphNode_StateMachine_0.Main_NoMusic.AnimStateTransitionNode_1.Transition
    src=Play
    via=AnimStateTransitionNode_1
   }
  ]
 }
 Idle {
  path=AnimGraph.AnimGraphNode_StateMachine_0.Main_NoMusic.AnimStateNode_0.Idle
  nodes [Output Animation Pose(Result=Sequence Player(Sequence=Get BaseAnim_NoMusic() -> (BaseAnim_NoMusic:Anim Sequence Base Object Reference)))]
 }
 Play {
  path=AnimGraph.AnimGraphNode_StateMachine_0.Main_NoMusic.AnimStateNode_1.Play
  nodes [
   n1 = Get IdleBreakers() -> (IdleBreakers:Array of Anim Sequence Base Object References)
   n2 = Get BreakerIndex() -> (BreakerIndex:int=0)
   Output Animation Pose(Result=Sequence Player!(
     Sequence="Get (a copy)"(Array:Array of Anim Sequence Base Object References=@n1, Dimension 1=@n2) -> (Output:Anim Sequence Base Object Reference)
     bLoopAnimation:bool=false
    ) -> (Pose))
  ]
 }
 Transition {
  path=AnimGraph.AnimGraphNode_StateMachine_0.Main_NoMusic.AnimStateTransitionNode_0.Transition
  nodes [Result(bCanEnterTransition=Get PlayBreaker() -> (PlayBreaker:bool))]
 }
 Transition {
  path=AnimGraph.AnimGraphNode_StateMachine_0.Main_NoMusic.AnimStateTransitionNode_1.Transition
  nodes [
   n1 = "Get (a copy)"(Array=Get IdleBreakers(), Dimension 1=Get BreakerIndex() -> (BreakerIndex=0))
   n2 = "Get Relevant Anim Time Remaining (Play)"!() -> (ReturnValue:float) {is=title K2Node_AnimGetter}
   Result(bCanEnterTransition=OR Boolean!(A="float <= float"(A=@n2) -> (ReturnValue=~), B="Equal (Object)"(A=@n1), C:bool=Get StopBreaker() -> (StopBreaker:bool)) -> (ReturnValue))
  ]
 }
 Main {
  entry=Idle
  path=AnimGraph.AnimGraphNode_StateMachine_1.Main
  ids=id
  nodes [
   AnimStateNode_0 = Idle()
   AnimStateNode_1 = Play()
   AnimStateTransitionNode_0 = AnimStateTransitionNode() {bAllowInertializationForSelfTransitions=true}
   AnimStateTransitionNode_1 = AnimStateTransitionNode() {bAllowInertializationForSelfTransitions=true}
  ]
  links [
   {dst=Play, rule=AnimGraph.AnimGraphNode_StateMachine_1.Main.AnimStateTransitionNode_0.Transition, src=Idle, via=AnimStateTransitionNode_0}
   {dst=Idle, rule=AnimGraph.AnimGraphNode_StateMachine_1.Main.AnimStateTransitionNode_1.Transition, src=Play, via=AnimStateTransitionNode_1}
  ]
 }
 Idle {path=AnimGraph.AnimGraphNode_StateMachine_1.Main.AnimStateNode_0.Idle, nodes [Output Animation Pose(Result=Use cached pose 'ChacheIdle'())]}
 Play {
  path=AnimGraph.AnimGraphNode_StateMachine_1.Main.AnimStateNode_1.Play
  nodes [
   n1 = "Get (a copy)"(Array=Get IdleBreakers(), Dimension 1=Get BreakerIndex() -> (BreakerIndex=0))
   Play Beat Synced Animation(
    ExplicitTime=0
    bShouldLoop=True
    InSequence=@n1
    TempoMultiplier=1
    bEnableTempoVariants=False
    TempoVariantHysteresis=0.05
    bShowDebug=False
   )
   n2 = Layered blend per bone(BasePose=Use cached pose 'ChacheIdle'(), BlendPoses_0=Sequence Player!(Sequence=@n1) -> (Pose) {bLoopAnimation=false}) {layerSetup [{branchFilters [
       {boneName=tom_mel_clavicle_l}
       {boneName=tom_mel_clavicle_r}
       {boneName=tom_mel_leg_l_01}
       {boneName=tom_mel_leg_r_01}
       {boneName=tom_ryan_clavicle_l}
       {boneName=tom_ryan_clavicle_r}
       {boneName=tom_ryan_leg_l_01}
       {boneName=tom_ryan_leg_r_01}
      ]}]}
   Output Animation Pose(Result=@n2)
  ]
 }
 Transition {path=AnimGraph.AnimGraphNode_StateMachine_1.Main.AnimStateTransitionNode_0.Transition, nodes [Result(bCanEnterTransition=Get PlayBreaker())]}
 Transition {
  path=AnimGraph.AnimGraphNode_StateMachine_1.Main.AnimStateTransitionNode_1.Transition
  nodes [
   n1 = "Get (a copy)"(Array=Get IdleBreakers(), Dimension 1=Get BreakerIndex() -> (BreakerIndex=0))
   n2 = "Get Relevant Anim Time Remaining (Play)"()
   Result(bCanEnterTransition=OR Boolean(A="float <= float"(A=@n2) -> (ReturnValue=~), B="Equal (Object)"(A=@n1), C=Get StopBreaker()))
  ]
 }
 EventGraph {nodes [
   Snare() {is=title K2Node_CustomEvent}
   > PrintString!(execute, InString=Snare, bPrintToScreen, bPrintToLog, TextColor, Duration, Key) -> (then)
   > SetParameter!(execute, self, ParameterName:Name=C_Hit, Value:float=GetGameTimeInSeconds()) -> (then)
   Tom 1() {is=title K2Node_CustomEvent}
   > PrintString(InString=Tom 1)
   > SetParameter(ParameterName=A_Hit, Value=GetGameTimeInSeconds())
   Randomize Faces() {is=title K2Node_CustomEvent}
   > n1 = For Loop(LastIndex=5)
   n2 = Select!(Option 0:String=A, Option 1:String=B, Option 2:String=C, Option 3:String=D, Option 4:String=E, Option 5:String=F, Index:int=@n1) -> (ReturnValue:String)
   > SetParameter(
    ParameterName=String To Name(InString=Append(A=@n2, B=_MouthIndex))
    Value="To Float (Integer)"(InInt=RandomInteger!(Max:int=5) -> (ReturnValue:int=0))
   )
   n3 = RandomInteger(Max=3) -> (ReturnValue=0)
   > SetParameter(
    ParameterName=String To Name(InString=Append(A=@n2, B=_EyesIndex))
    Value="To Float (Integer)"(InInt=Select!(Option 0:int=0, Option 1:int=1, Option 2:int=2, Option 3:int=5, Index=@n3) -> (ReturnValue:int))
   )
   ResetFaces() {is=title K2Node_CustomEvent}
   > n4 = For Loop(LastIndex=5)
   n5 = Select!(Option 0:String=A, Option 1:String=B, Option 2:String=C, Option 3:String=D, Option 4:String=E, Option 5:String=F, Index=@n4) -> (ReturnValue:String)
   > SetParameter(ParameterName=String To Name(InString=Append(A=@n5, B=_MouthIndex)), Value=2)
   > SetParameter(ParameterName=String To Name(InString=Append(A=@n5, B=_EyesIndex)), Value=2)
   Tom 2() {is=title K2Node_CustomEvent}
   > PrintString(InString=Tom 2)
   > SetParameter(ParameterName=B_Hit, Value=GetGameTimeInSeconds())
   Fakeclock() {is=title K2Node_CustomEvent}
   > Do Once()
   > n6 = Delay(Duration=1)
   Set FakeClockValid(execute=[@n6.then, @n9.then], FakeClockValid:bool=true) -> (Output_Get:bool)
   > n7 = Sequence()
   > Delay(Duration=7)
   > Set FakeClockValid(FakeClockValid=false)
   > n8 = Sequence()
   > n9 = Delay(Duration=4)
   ResetFaces(execute=@n8.then_1) {is=title K2Node_CallFunction}
   Randomize Faces(execute=@n7.then_1) {is=title K2Node_CallFunction}
   Tom Floor() {is=title K2Node_CustomEvent}
   > PrintString(InString=Tom Floor)
   > SetParameter(ParameterName=D_Hit, Value=GetGameTimeInSeconds())
   Hi Hat() {is=title K2Node_CustomEvent}
   > PrintString(InString=Hi Hat)
   > SetParameter(ParameterName=E_Hit, Value=GetGameTimeInSeconds())
   Cymbal() {is=title K2Node_CustomEvent}
   > PrintString(InString=Ride)
   > SetParameter(ParameterName=F_Hit, Value=GetGameTimeInSeconds())
   GetClock() {is=title K2Node_CustomEvent}
   n10 = GetHighestPriorityClock!(self:Music Event Subsystem Object Reference=Get MusicEventSubsystem() -> (ReturnValue:Music Event Subsystem Object Reference) {is=title K2Node_GetSubsystem}) -> (ReturnValue:Music Clock Object Reference)
   > Set Clock(Clock:Music Clock Object Reference=@n10) -> (Output_Get:Music Clock Object Reference)
   n11 = GetCurrentTempo!(self:Music Clock Object Reference=@n10) -> (ReturnValue:float)
   > n12 = Set ClockValid(ClockValid:bool=AND Boolean(A=IsValid(Object=@n10), B="Not Equal (Float)"(A=@n11))) -> (Output_Get:bool)
   > n13 = Branch(Condition=OR Boolean!(A=@n12, B=Get FakeClockValid()) -> (ReturnValue))
   > Do Once(Reset=@n15.then_1)
   > n14 = Sequence()
   > Randomize Faces()
   Do Once(execute=@n13.else, Reset=@n14.then_1)
   > n15 = Sequence()
   > ResetFaces()
   Event BlueprintUpdateAnimation()
   > n16 = Sequence!(execute) -> (then_0, then_1, then_2:Exec, then_3:Exec, then_4:Exec, then_5:Exec)
   Branch(execute=@n16.then_1, Condition=Get PlayBreaker())
   > Do Once(Reset=@n17.then)
   > DelayUntilNextTick!(execute) -> (then)
   > n17 = Set PlayBreaker(PlayBreaker:bool=false) -> (Output_Get:bool)
   GetClock(execute=@n16.then_2) {is=title K2Node_CallFunction}
   Branch(execute=@n16.then_3, Condition=Get StopBreaker())
   > Do Once(Reset=@n18.then)
   > DelayUntilNextTick()
   > n18 = Set StopBreaker(StopBreaker:bool=false) -> (Output_Get:bool)
   n19 = Branch(execute=@n16.then_4, Condition=OR Boolean(A=Get ClockValid(), B=Get FakeClockValid()))
   > n20 = Sequence!(execute) -> (then_0, then_1)
   Do Once(execute=@n23.then_0, Reset=@n20.then_0)
   > n21 = For Loop(LastIndex=5)
   > Set Array Elem!(
    execute
    TargetArray:"Array of Float (double-precision)s"=Get NewVar_1() -> (NewVar_1:"Array of Float (double-precision)s")
    Index:int=@n21
    Item:"Float (double-precision) (by ref)"=RandomFloatInRange(Min=2, Max=6)
    bSizeToFit:bool=false
   ) -> (then)
   n22 = Select(Option 0=A, Option 1=B, Option 2=C, Option 3=D, Option 4=E, Option 5=F, Index=@n21)
   > SetParameter(ParameterName=String To Name(InString=Append(A=@n22, B=_EyesIndex)), Value=2)
   n23 = Sequence(execute=@n19.else)
   n24 = For Loop(execute=@n23.then_1, LastIndex=5)
   n25 = "Get (a copy)"(Array:"Array of Float (double-precision)s"=Get NewVar_0() -> (NewVar_0:"Array of Float (double-precision)s"), Dimension 1=@n24) -> (Output:double)
   > Set Array Elem(TargetArray=Get NewVar_0(), Index=@n24, Item=float + float(A=@n25, B=GetWorldDeltaSeconds!() -> (ReturnValue:double)), bSizeToFit=false)
   n26 = "Get (a copy)"(Array=Get NewVar_1(), Dimension 1=@n24)
   > Branch(Condition="float >= float"(A=@n25, B=@n26))
   > n27 = Branch(Condition="float >= float"(A=@n25, B=float + float(A=@n26, B=0.2)))
   n28 = String To Name(InString=Append(A=Select(Option 0=A, Option 1=B, Option 2=C, Option 3=D, Option 4=E, Option 5=F, Index=@n24), B=_EyesIndex))
   > SetParameter(ParameterName=@n28, Value=2)
   > Set Array Elem(TargetArray=Get NewVar_1(), Index=@n24, Item=RandomFloatInRange(Min=2, Max=6), bSizeToFit=false)
   > Set Array Elem(TargetArray=Get NewVar_0(), Index=@n24, bSizeToFit=false)
   SetParameter(execute=@n27.else, ParameterName=@n28, Value=3)
   AnimNotify_SonnetSpirit_FaceChange_01() {is=title K2Node_Event}
   > Branch(Condition=OR Boolean(A=Get ClockValid(), B=Get FakeClockValid()))
   > n29 = Sequence()
   > Set RandomEyeIndex(RandomEyeIndex:int=RandomInteger(Max=3) -> (ReturnValue=0)) -> (Output_Get:int=0)
   n30 = Get IsWiggleing() -> (IsWiggleing:bool)
   > n31 = Set FacePool(FacePool:Array of Strings=Make Array!(
     "[0]":String=SelectString(A=D, B=A, bPickA=@n30)
     "[1]":String=SelectString(A=C, B=B, bPickA=@n30)
     "[2]":String=C
     "[3]":String=D
     "[4]":String=E
     "[5]":String=F
    ) -> (Array:Array of Strings)) -> (Output_Get:Array of Strings)
   > Remove Item!(execute, TargetArray=@n31, Item:"String (by ref)"=Get LastEyesFace() -> (LastEyesFace:String)) -> (then, ReturnValue:bool)
   > n32 = Set LastEyesFace(LastEyesFace:String=Random Array Item!(TargetArray=@n31) -> (OutItem:String, OutIndex:int=0) {is=title K2Node_CallArrayFunction}) -> (Output_Get:String)
   n33 = String To Name(InString=Append(A=@n32, B=_EyesIndex))
   > n34 = GetParameter!(execute, self, ParameterName:Name=@n33) -> (then, Value:float)
   n35 = SelectInt(A=2, B=1, bPickA="Equal (Float)"(A=@n34))
   n36 = SelectInt(A=2, B=0, bPickA="Equal (Float)"(A=@n34, B=1))
   n37 = Get RandomEyeIndex() -> (RandomEyeIndex:int=0)
   > SetParameter(
    ParameterName=@n33
    Value="To Float (Integer)"(InInt=Select!(Option 0:int=@n36, Option 1:int=@n35, Option 2:int=2, Option 3:int=5, Index=@n37) -> (ReturnValue:int))
   )
   Delay(execute=@n29.then_1, Duration=0.1)
   > n38 = Set FacePool(FacePool=Make Array!("[0]"=A, "[1]"=B, "[2]"=C, "[3]"=D) -> (Array))
   > Remove Item(TargetArray=@n38, Item=Get LastMouthFace() -> (LastMouthFace:String))
   > n39 = Set LastMouthFace(LastMouthFace:String=Random Array Item(TargetArray=@n38) -> (OutIndex=0)) -> (Output_Get:String)
   > SetParameter(
    ParameterName=String To Name(InString=Append(A=@n39, B=_MouthIndex))
    Value="To Float (Integer)"(InInt=RandomInteger(Max=5) -> (ReturnValue=0))
   )
   Event BlueprintInitializeAnimation()
   > n40 = Sequence()
   > n41 = Cast To FortPlayerPawn(Object=GetOwningActor())
   > Set Fort Player Pawn(Fort Player Pawn:Fort Player Pawn Object Reference=@n41) -> (Output_Get:Fort Player Pawn Object Reference)
   Branch(
    execute=@n40.then_1
    Condition=OR Boolean(
     A=IsCosmeticObjectDisplayedAsPreview!(CosmeticObject:Object Reference) -> (ReturnValue:bool)
     B=IsCosmeticObjectDisplayedInFrontendContext!(CosmeticObject:Object Reference) -> (ReturnValue:bool)
    )
   )
   > Fakeclock() {is=title K2Node_CallFunction}
   AnimNotify_BlendOut() {is=title K2Node_Event}
   n42 = "To Float (Integer)"(InInt=Select(Option 0=0, Option 1=1, Option 2=2, Option 3=5, Index=RandomInteger(Max=3) -> (ReturnValue=0)))
   > SetParameter(ParameterName=A_EyesIndex, Value=@n42)
   > SetParameter(ParameterName=B_EyesIndex, Value=@n42)
   > Set IsWiggleing(IsWiggleing:bool=false) -> (Output_Get:bool)
   AnimNotify_Wiggle() {is=title K2Node_Event}
   n43 = Get CountDownWiggle() -> (CountDownWiggle:int=0)
   > n44 = Branch(Condition="integer <= integer"(A=@n43))
   > Set CountDownWiggle(CountDownWiggle:int=RandomIntegerInRange!(Min:int=3, Max:int=5) -> (ReturnValue:int=0)) -> (Output_Get:int=0)
   > Set PlayBreaker(PlayBreaker=true)
   > Set IsWiggleing(IsWiggleing=true)
   > SetParameter(ParameterName=A_EyesIndex, Value=4)
   > SetParameter(ParameterName=B_EyesIndex, Value=4)
   Decrement Int(" "=@n44.else, Value=@n43)
   n45 = Get!(execute:Exec) -> (then:Exec, else:Exec, Fort Player Pawn:Fort Player Pawn Object Reference)
   n46 = GetAnimInstance(self=Get Mesh(self=@n45))
   > Branch(Condition=Does Object Implement Interface!(
     TestObject:Object Reference=@n46
     Interface:Interface Interface=/Game/Characters/Player/Male/Male_Avg_Base/AnimLayerInterface/PlayerAnimOverrideLayerInterface.PlayerAnimOverrideLayerInterface_C
    ) -> (ReturnValue:bool))
   > SetDrumsEventAndStateInterfaceByClass!(
    execute
    self
    AnimLayerClass:Fort Base Layer Anim Instance Soft Class Reference=Get BandPlayDrumsClass() -> (BandPlayDrumsClass:Fort Base Layer Anim Instance Soft Class Reference)
    PlayerAnimInstance:Anim Instance Object Reference=@n46
   ) -> (then, Success:bool)
   n47 = Get SparksDrumEventsInterface() -> (SparksDrumEventsInterface:Sparks Drum Events Interface Interface)
   > Bind_SnareHit!(execute, self:Sparks Drum Events Interface Interface=@n47, Callback:"Delegate (by ref)"=Create Event()) -> (then)
   > Bind_Tom1_Hit!(execute, self:Sparks Drum Events Interface Interface=@n47, Callback:"Delegate (by ref)"=Create Event()) -> (then)
   > Bind_Tom2_Hit!(execute, self:Sparks Drum Events Interface Interface=@n47, Callback:"Delegate (by ref)"=Create Event()) -> (then)
   > Bind_FloorTomHit!(execute, self:Sparks Drum Events Interface Interface=@n47, Callback:"Delegate (by ref)"=Create Event()) -> (then)
   > Bind_HiHatHit!(execute, self:Sparks Drum Events Interface Interface=@n47, Callback:"Delegate (by ref)"=Create Event()) -> (then)
   > Bind_RideHardHit!(execute, self:Sparks Drum Events Interface Interface=@n47, Callback:"Delegate (by ref)"=Create Event()) -> (then)
  ]}
 GetFace {nodes [
   n1 = GetFace!() -> (then, LastFace:String) {is=title K2Node_FunctionEntry}
   n2 = Get FacesPool() -> (FacesPool:Array of Strings)
   > Remove Item(TargetArray=@n2, Item=@n1)
   > Return Node!(execute, Output_Get:String=Random Array Item(TargetArray=@n2) -> (OutIndex=0))
  ]}
 GetParameter {nodes [
   n1 = GetParameter!() -> (then, ParameterName:Name) {is=title K2Node_FunctionEntry}
   > Branch(Condition=Get UseHitReactivityInFlipbook() -> (UseHitReactivityInFlipbook:bool))
   > n2 = Is Valid(InputObject=Get MatInst() -> (MatInst:Material Instance Dynamic Object Reference))
   n3 = GetOwningComponent()
   Is Valid(exec=@n2.Is Not Valid, InputObject=@n3)
   > n4 = Get Dynamic Material Instance(SkeletalMeshComponent=@n3)
   > n5 = Set MatInst(MatInst:Material Instance Dynamic Object Reference=@n4) -> (Output_Get:Material Instance Dynamic Object Reference)
   n6 = Get Scalar Parameter Value!(execute=[@n2.Is Valid, @n5.then], self:Material Instance Dynamic Object Reference=@n5, ParameterName:Name=@n1) -> (then, ReturnValue:float) {is=title K2Node_CallFunction}
   > Return Node!(execute, Value:float=@n6)
  ]}
 SetDrumsEventAndStateInterfaceByClass {nodes [
   n1 = SetDrumsEventAndStateInterfaceByClass!() -> (then, AnimLayerClass:Fort Base Layer Anim Instance Soft Class Reference, PlayerAnimInstance:Anim Instance Object Reference) {is=title K2Node_FunctionEntry}
   n2 = Resolve Soft Reference!(Input:Fort Base Layer Anim Instance Soft Class Reference=@n1) -> (Output:Fort Base Layer Anim Instance Class Reference)
   > n3 = Branch(Condition=IsValidClass!(Class:Object Class Reference=@n2) -> (ReturnValue:bool))
   > n4 = Cast To SparksBaseLayerAnimInstance!(
    execute
    Object=GetLinkedAnimLayerInstanceByClass!(self:Anim Instance Object Reference=@n1.PlayerAnimInstance, InClass:Anim Instance Class Reference=@n2, bCheckForChildClass:bool=true) -> (ReturnValue:Anim Instance Object Reference)
   ) -> (then, CastFailed, AsSparks Base Layer Anim Instance:Sparks Base Layer Anim Instance Object Reference) {is=title K2Node_DynamicCast}
   > Set BandPlayLayer(BandPlayLayer:Sparks Base Layer Anim Instance Object Reference=@n4) -> (Output_Get:Sparks Base Layer Anim Instance Object Reference)
   > n5 = Get!(execute) -> (then, else, BandPlayLayer:Sparks Base Layer Anim Instance Object Reference)
   n6 = Get BandPlayLayer() -> (BandPlayLayer:Sparks Base Layer Anim Instance Object Reference)
   > Set LinkedAnimGraphValue(LinkedAnimGraphValue:Anim Instance Object Reference=GetLinkedAnimGraphInstanceByTag!(
     self:Anim Instance Object Reference=@n6
     InTag:Name=Get InstrumentSlotTag!(self:Sparks Base Layer Anim Instance Object Reference=@n6) -> (InstrumentSlotTag:Name=None)
    ) -> (ReturnValue:Anim Instance Object Reference)) -> (Output_Get:Anim Instance Object Reference)
   > n7 = Branch(Condition=IsValid(Object=Get LinkedAnimGraphValue() -> (LinkedAnimGraphValue:Anim Instance Object Reference)))
   > n8 = Cast To SparksDrumEventsInterface!(execute, Object=Get LinkedAnimGraphValue()) -> (then, CastFailed, AsSparks Drum Events Interface:Sparks Drum Events Interface Interface) {is=title K2Node_DynamicCast}
   > Set SparksDrumEventsInterface(SparksDrumEventsInterface:Sparks Drum Events Interface Interface=@n8) -> (Output_Get:Sparks Drum Events Interface Interface)
   > n9 = Cast To Sparks_Drum_AnimBP_StateInterface!(execute, Object=Get LinkedAnimGraphValue()) -> (then, CastFailed, AsSparks Drum Anim BP State Interface:Sparks Drum Anim BP State Interface Interface) {is=title K2Node_DynamicCast}
   > Set SparksDrumStateInterface(SparksDrumStateInterface:Sparks Drum Anim BP State Interface Interface=@n9) -> (Output_Get:Sparks Drum Anim BP State Interface Interface)
   > Set SparksDrumsAnimBPValid(SparksDrumsAnimBPValid:bool=true) -> (Output_Get:bool)
   > Return Node!(execute, Success:bool=true)
   Return Node(execute=[@n9.CastFailed, @n8.CastFailed, @n7.else, @n5.else, @n4.CastFailed, @n3.else], Success=false)
   Reroute Node(InputPin:Wildcard) -> (OutputPin:Wildcard)
  ]}
 SetParameter {nodes [
   n1 = SetParameter!() -> (then, ParameterName:Name, Value:float=~) {is=title K2Node_FunctionEntry}
   > Branch(Condition=Get UseHitReactivityInFlipbook())
   > n2 = Is Valid(InputObject=Get MatInst())
   n3 = GetOwningComponent()
   Is Valid(exec=@n2.Is Not Valid, InputObject=@n3)
   > n4 = Get Dynamic Material Instance(SkeletalMeshComponent=@n3)
   > n5 = Set MatInst(MatInst=@n4)
   SetScalarParameterValue(execute=[@n2.Is Valid, @n5.then], self=@n5, ParameterName=@n1, Value=@n1.Value)
  ]}
]
variables [
 BandPlayDrumsClass {type=Class, default=/Script/SparksCharacterCommonRuntime.SparksBaseLayerAnimInstance}
 BaseAnim {type=AnimSequenceBase, default=&/SonnetSpirit_DrumKit_Idle}
 BaseAnim_NoMusic {type=AnimSequenceBase, default=&/SonnetSpirit_DrumKit_Idle_NoMusic}
 Blinks {type="array<boolean>", default []}
 BreakerIndex {type=integer}
 Clock {type=MusicClockComponent, default=null}
 ClockValid {type=boolean}
 CountDownWiggle {type=integer}
 FacePool {type="array<string>", default [A,B,C,D,E,F]}
 FakeClockValid {type=boolean}
 Fort Player Pawn {type=FortPlayerPawn}
 IdleBreakers {
  type="array<AnimSequenceBase>"
  default [&/SonnetSpirit_DrumKit_NoodleArm_fidget]
 }
 IsWiggleing {type=boolean}
 LastEyesFace {type=string, default=("A","B","C","D","E","F")}
 LastMouthFace {type=string, default=("A","B","C","D","E","F")}
 MatInst {type=MaterialInstanceDynamic}
 NewVar {type=boolean}
 NewVar_0 {type="array<number>", default [0,0,0,0,0,0]}
 NewVar_1 {type="array<number>", default [0,0,0,0,0,0]}
 PlayBreaker {type=boolean}
 RandomEyeIndex {type=integer}
 RandomFace {type="array<integer>", default [0,1,2,3,4,5]}
 RandomFaceIndex {type=integer}
 SparksDrumEventsInterface {default=None}
 SparksDrumsAnimBPValid {type=boolean}
 SparksDrumStateInterface {default=None}
 StopBreaker {type=boolean}
 UseHitReactivityInFlipbook {type=boolean, default=true}
]
